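# Sound cue for the fill-gap sound of the soccer traversal emote.
# A Mixer layers a random pick of two fill-gap waves, passed through a Modulator, with a louder step wave that starts after a short fixed delay.
# It uses Athena emote attenuation, the emote music sound class and concurrency rule, and a send to the music audio-analysis submix.
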SoundCue /Game/Athena/Sounds/Emotes/SoccerTraversal/Emote_SoccerTraversal_FillGap_Cue
root=Mixer_0
nodes [
 {id=Delay_0, children [WavePlayer_3], properties {delayMax=0.25, delayMin=0.25}}
 {id=Mixer_0, children [Modulator_0,Delay_0], properties {inputVolume [1,1.3]}}
 {id=Modulator_0, children [Random_0], properties {volumeMax=1, volumeMin=1}}
 {id=Random_0, children [WavePlayer_1,WavePlayer_2], properties {weights [1,1]}}
 {id=WavePlayer_1, properties {soundWaveAssetPtr=&/Emote_SoccerTraversal_FillGap_01}}
 {id=WavePlayer_2, properties {soundWaveAssetPtr=&/Emote_SoccerTraversal_FillGap_02}}
 {id=WavePlayer_3, properties {soundWaveAssetPtr=&/Emote_SoccerTraversal_FillGap_Step_01}}
]
properties {
 attenuationSettings=&Fort_Emotes_Athena_Att
 bOverrideConcurrency=true
 maxDistance=1600
 soundClassObject=&Fort_Music_Emote
 volumeMultiplier=1.7
 concurrencySet [/Game/Sounds/Concurrency_Rules/Athena_EmoteMusic_Conc]
 soundSubmixSends [{minSendLevel=0, soundSubmix=&Fortnite_AudioAnalysis_Music}]
}
